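# Niagara effect for the PeacefulPoemPumped emote: a flickering purple spark burst with flare, glow, smoke and debris.
# Paired flare and glow sprites fade on the system's looped age, with FlickerSpeed driving flare brightness, light and spark rate.
# Sparks and smoke launch in cones along system direction variables, while physical-surface debris meshes use a fixed surface and direction.
# Five other emitters, including montage-section directional sparks and warm-colored flashes, are disabled.
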NiagaraSystem /BRCosmetics/Effects/Fort_Effects/Emotes/PeacefulPoemPumped/NS_Emote_PeacefulPoemPumped_Sparks
dependencies {
 usedDataInterfaces [
  &NiagaraDataInterfaceArrayFloat3
  &NiagaraDataInterfaceCurve
  &NiagaraDataInterfaceColorCurve
  &NiagaraDataInterfaceVectorCurve
  &NiagaraDataInterfaceMeshRendererInfo
  &NiagaraDataInterfaceVector2DCurve
 ]
 usedDynamicInputs [&SimulationPosition, &Waveform]
 usedModules [
  &SystemState
  &self:SystemSpawnScript.SystemScriptSource.SystemScriptGraph.NiagaraNodeAssignment_0.SetVariables_#1
  &self:SystemSpawnScript.SystemScriptSource.SystemScriptGraph.NiagaraNodeAssignment_1.SetVariables_#2
 ]
}
emitters {
 DirectionalSparks {
  bEnabled=false
  emitterUpdateScript [
   &EmitterState()
   &SpawnBurst_Instantaneous(
    Spawn Count=&UniformRangedInt(Minimum=9, Maximum=12)
    Spawn Probability=&FN_FrontEndFloatModifier(Gameplay=1, Front End=0)
    Use Spawn Probability=true
    UseLoopCountLimit=false
   )
  ]
  particleSpawnScript [
   &InitializeParticle(
    Lifetime Max=0.45
    Lifetime Min=0.05
    Lifetime Mode=::NewEnumerator1
    Sprite Size Mode=::NewEnumerator1
    Uniform Sprite Size Max=5
    Uniform Sprite Size Min=3
    UsePositionOffset=false
    Write Material Random=true
   )
   &ShapeLocation(
    Cylinder Height=80
    Cylinder Radius=2
    MaskSpawnGroup=false
    Override Seed=false
    Rotation Axis=/Game/Effects/Niagara/DynamicInputs/FN_DirectionalVectorfromMontageIndex(
     Montage Section Index:NiagaraInt32=$User.SwingSectionIndex
     Montage Index 1:Vector3f=<0 1 1>
     Montage Index 2:Vector3f=<0 1 0.6>
     Montage Index 3:Vector3f=<1 -1 0.36>
     Montage Index 4:Vector3f=<1 -0.5 -1>
     Invalid Montage Index:Vector3f=<1 0 0>
    )
    Rotation Mode=::NewEnumerator1
    Shape Primitive=::NewEnumerator1
    Surface Only=false
    UseOwnerScale=true
   ) {enabled=false}
   &ShapeLocation(
    Additional Quaternion Rotation=false
    Cylinder Height=60
    Cylinder Height Midpoint=0.25
    Cylinder Radius=2
    MaskSpawnGroup=false
    Override Seed=false
    Rotation Mode=::NewEnumerator3
    Rotation Quaternion=&OrientMeshToVector(
     Look At Direction=&TransformVector(
      Vector=&FN_DirectionalVectorfromMontageIndex(
       Montage Section Index:NiagaraInt32=$User.SwingSectionIndex
       Montage Index 1:Vector3f=<0.6944 1 1>
       Montage Index 2:Vector3f=<-1 1 0.9506>
       Montage Index 3:Vector3f=<2.6556 -1 2.3295>
       Montage Index 4:Vector3f=<2.4576 -1 -4.0592>
       Invalid Montage Index:Vector3f=<1 0 0>
      )
      Destination Space=::World
     )
     Reference Axis Vector=<0 0 1>
    )
    Shape Primitive=::NewEnumerator1
    Surface Only=false
    UseOwnerScale=true
   ) {enabled=false, module=ShapeLocation001}
   &AddVelocity(
    Cone Angle=20
    Cone Axis=&RotateVectorByQuaternion(
     Vector To Rotate=<0 0 1>
     Quaternion=&OrientMeshToVector(
      Look At Direction=&TransformVector(
       Vector=&FN_DirectionalVectorfromMontageIndex(
        Montage Section Index:NiagaraInt32=$User.SwingSectionIndex
        Montage Index 1:Vector3f=<0 1 -1>
        Montage Index 2:Vector3f=<0 -1 -1>
        Montage Index 3:Vector3f=<0 1 -1>
        Montage Index 4:Vector3f=<0 -1 -1>
        Invalid Montage Index:Vector3f=<1 0 0>
       )
       Destination Space=::World
      )
      Reference Axis Vector=<0 0 1>
     )
    )
    Module.UseConeFalloff=false
    Module.UseVelDistribution=false
    Override Seed=false
    Rotation Coordinate Space=::World
    Velocity Mode=::NewEnumerator2
    Velocity Speed=&RandomRangeFloat(Minimum=400, Maximum=1000)
   ) {module=AddVelocity001}
   &CameraOffset(Camera Offset Amount=&RandomRangeFloat(Minimum=50, Maximum=100))
   &AddVelocityFromPoint(Module.Use Falloff Distance=false, Velocity Strength=&RandomRangeFloat(Minimum=150, Maximum=200)) {enabled=false}
   &AddVelocity(
    Velocity=&FN_DirectionalVectorfromMontageIndex(
     Montage Section Index:NiagaraInt32=$User.SwingSectionIndex
     Montage Index 1:Vector3f=<0 1 1>
     Montage Index 2:Vector3f=&RandomRangeVector(Minimum=<0 1 0.5>, Maximum=<0 1 1>)
     Montage Index 3:Vector3f=&RandomRangeVector(Minimum=<1 -1 0.5>, Maximum=<1 -1 0.7>)
     Montage Index 4:Vector3f=&RandomRangeVector(Minimum=<1 -0.5 -1>, Maximum=<1 -0.7 -1>)
     Invalid Montage Index:Vector3f=<1 0 0>
    )
    Velocity Speed Scale=&RandomRangeFloat(Minimum=300, Maximum=900)
   ) {enabled=false}
  ]
  particleUpdateScript [
   &ParticleState()
   &AccelerationForce(Acceleration=&RandomRangeVector(Minimum=<0 0 -500>, Maximum=<0 0 -800>))
   &Drag(Drag=&RandomRangeFloat(Minimum=5, Maximum=8), Use Linear Drag=true, Use Rotational Drag=true)
   &SolveForcesAndVelocity(Clamp Velocity=false, Limit Acceleration=false)
   &ScaleSpriteSizeBySpeed(Max Scale Factor=<0.5 10>, Min Scale Factor=<0.5 0.5>)
   &ScaleSpriteSize(Uniform Curve Sprite Scale=DataInterface() {propertyValues {Curve {keys [{arriveTangent=-6.725, interpMode=RCIM_Cubic, leaveTangent=-6.725, tangentMode=RCTM_User, time=0.65, value=1}, <at 1 0 RCIM_Cubic>]}}})
   &ScaleColor(
    Linear Color Curve=DataInterface() {propertyValues {
      AlphaCurve {keys [<at 0 0>,<at 0.166 1>,<at 1 0>]}
      BlueCurve {keys [<at 0 0.24>,<at 0.166 0.2>,<at 0.333 0>,<at 0.75 0>]}
      GreenCurve {keys [<at 0 3.12>,<at 0.166 2.6>,<at 0.333 0.6667>,<at 0.75 0.15>]}
      RedCurve {keys [<at 0 6>,<at 0.166 5>,<at 0.333 4>,<at 0.75 3>]}
     }}
    Scale Mode=::NewEnumerator2
   ) {module=ScaleColor001}
  ]
  rendererClasses [&NiagaraSpriteRendererProperties]
  renderers [{rendererClass=&NiagaraSpriteRendererProperties}]
 }
 Flash_Flare {
  emitterUpdateScript [
   &EmitterState()
   &SpawnBurst_Instantaneous(Spawn Count=2, Use Spawn Probability=false)
  ]
  particleSpawnScript [
   &PartitionParticles()
   &InitializeParticle(
    Color=<rgba 0.3968 0.0137 0.7084 1>
    Lifetime=0.1
    Mass Mode=::NewEnumerator1
    Position Offset=&SelectVectorFromArray(
     Direct Array Index=$Particles.PartitionParticles.Partition
     Vector Selection Array=DataInterface() {propertyValues {FloatData [<0 -2.5 0>, <0 2.5 0>]}}
    )
    Sprite Size=<65 65>
    Sprite Size Mode=::NewEnumerator4
    UsePositionOffset=true
    Write Material Random=true
   ) {module=InitializeParticle001}
   &CameraOffset(Camera Offset Amount=10)
   &DynamicMaterialParameters(
    Index 0 Param 1=0.2
    Module.Param0WriteEnabled=true
    Module.Param1WriteEnabled=true
    Module.Param2WriteEnabled=true
    Module.Param3WriteEnabled=true
   )
  ]
  particleUpdateScript [
   &ParticleState(
    Kill Particles When Lifetime Has Elapsed=false
    Let Infinitely Lived Particles Die When Emitter Deactivates=true
    Loop Particles Lifetime=true
   )
   &SolveForcesAndVelocity(Clamp Velocity=false, Limit Acceleration=false)
   &ScaleColor(
    Scale Alpha=&FloatFromCurve(
     FloatCurve=DataInterface() {propertyValues {Curve {keys [
         {arriveTangent=17.213, interpMode=RCIM_Cubic, leaveTangent=17.213, tangentMode=RCTM_User, time=0, value=0}
         {arriveTangent=0.3861, interpMode=RCIM_Cubic, leaveTangent=0.3861, tangentMode=RCTM_User, time=0.1, value=0.9847}
         {arriveTangent=-4.1234, interpMode=RCIM_Cubic, leaveTangent=-4.1234, tangentMode=RCTM_User, time=1, value=0}
        ]}}}
     CurveIndex=$System.LoopedAge
    )
    Scale RGB=&VectorFromFloat(Value=&Multiply_Float(A=8, B=$System.FlickerSpeed))
    ScaleA=true
    ScaleRGB=true
   )
   &FN_Light_v2(
    Light Exponent=3
    Light Radius=60
    LightEnabled=true
    Position Offset:Vector3f=<-15 0 0>
    Scale Color:NiagaraFloat=&Multiply_Float(A=1.5, B=$System.FlickerSpeed)
   )
  ]
  rendererClasses [&NiagaraSpriteRendererProperties, &NiagaraLightRendererProperties]
  renderers [
   {rendererClass=&NiagaraSpriteRendererProperties}
   {rendererClass=&NiagaraLightRendererProperties, rendererIndex=1}
  ]
 }
 Flash_Flare001 {
  bEnabled=false
  emitterUpdateScript [
   &EmitterState()
   &SpawnBurst_Instantaneous(Use Spawn Probability=false, UseLoopCountLimit=false)
  ]
  particleSpawnScript [
   &InitializeParticle(
    Color=<rgba 1.5 0.755 0.266 1>
    Lifetime=0.1
    Sprite Size=&Multiply_Vector2D_ByFloat(Float=&FN_FrontEndFloatModifier(Gameplay=1, Front End=$User.FE_Scale), Vector2D=<200 120>)
    Sprite Size Mode=::NewEnumerator4
    UsePositionOffset=false
    Write Material Random=true
   )
   &DynamicMaterialParameters(
    Index 0 Param 1=0.2
    Index 0 Param 2=1
    Module.Param0WriteEnabled=true
    Module.Param1WriteEnabled=true
    Module.Param2WriteEnabled=true
    Module.Param3WriteEnabled=true
   )
   &SystemLocation(MaskSpawnGroup=false, Offset=$System.HitOffset)
  ]
  particleUpdateScript [
   &ParticleState()
   &ScaleColor(
    Scale Alpha=&Lerp_Float(A=1, B=0, Alpha=$Particles.NormalizedAge)
    Scale RGB=&VectorFromFloat(Value=2)
    ScaleA=true
    ScaleRGB=true
   )
   &FN_Light_v2(Light Exponent=3, Light Radius=150, LightEnabled=true, Scale Color=2)
  ]
  rendererClasses [&NiagaraSpriteRendererProperties, &NiagaraLightRendererProperties]
  renderers [
   {rendererClass=&NiagaraSpriteRendererProperties}
   {rendererClass=&NiagaraLightRendererProperties, rendererIndex=1}
  ]
 }
 Glow {
  emitterUpdateScript [
   &EmitterState()
   &SpawnBurst_Instantaneous(Spawn Count=2, Use Spawn Probability=false)
  ]
  particleSpawnScript [
   &PartitionParticles()
   &InitializeParticle(
    Color=<rgba 0.07 0.0024 0.125 1>
    Lifetime=0.1
    Position Offset=&SelectVectorFromArray(
     Direct Array Index=$Particles.PartitionParticles.Partition
     Vector Selection Array=DataInterface() {propertyValues {FloatData [<0 -2.5 0>, <0 2.5 0>]}}
    )
    Sprite Size Mode=::NewEnumerator3
    Uniform Sprite Size=50
    UsePositionOffset=true
    Write Material Random=true
   ) {module=InitializeParticle001}
   &DynamicMaterialParameters(
    Index 0 Param 1=1
    Index 0 Param 2=1
    Index 0 Param 3=1
    Index 0 Param 4=1
    Module.Param0WriteEnabled=true
    Module.Param1WriteEnabled=true
    Module.Param2WriteEnabled=true
    Module.Param3WriteEnabled=true
   )
  ]
  particleUpdateScript [
   &CameraOffset()
   &ParticleState(Kill Particles When Lifetime Has Elapsed=false, Let Infinitely Lived Particles Die When Emitter Deactivates=true)
   &SolveForcesAndVelocity(Clamp Velocity=false, Limit Acceleration=false)
   &ScaleColor(
    Scale Alpha=&FloatFromCurve(
     FloatCurve=DataInterface() {propertyValues {Curve {keys [
         {arriveTangent=9.0725, interpMode=RCIM_Cubic, leaveTangent=9.0725, tangentMode=RCTM_User, time=0, value=0}
         {arriveTangent=0.3861, interpMode=RCIM_Cubic, leaveTangent=0.3861, tangentMode=RCTM_User, time=0.1408, value=0.9847}
         {arriveTangent=-4.1234, interpMode=RCIM_Cubic, leaveTangent=-4.1234, tangentMode=RCTM_User, time=1, value=0}
        ]}}}
     CurveIndex=$System.LoopedAge
     Scale Curve=&FNTimeOfDayScalar(Night Scale=0.2, Morning Scale=0.2, Day Scale=0.2, Evening Scale=0.2)
    )
    ScaleA=true
    ScaleRGB=true
   )
   &ScaleSpriteSize(Scale Factor=&Vector2DFromFloat(Value=&FloatFromCurve(FloatCurve=DataInterface() {propertyValues {Curve {keys [<at 0 1>,<at 1 3>]}}})))
  ]
  rendererClasses [&NiagaraSpriteRendererProperties]
  renderers [
   {rendererClass=&NiagaraSpriteRendererProperties}
   {rendererClass=&NiagaraSpriteRendererProperties, rendererIndex=1}
  ]
 }
 Lazy_Sparks {
  emitterUpdateScript [
   &EmitterState()
   &SpawnRate(Spawn Probability=0.67, SpawnRate=&Multiply_Float(A=150, B=$System.FlickerSpeed), Use Spawn Probability=true)
   &SpawnBurst_Instantaneous(Spawn Count=&UniformRangedInt(Minimum=15, Maximum=20), Spawn Time=0.05, Use Spawn Probability=false)
  ]
  particleSpawnScript [
   &InitializeParticle(
    Color=&MultiplyLinearColor(
     Linear Color=<rgba 1 1 1 1>
     Scale Factor=&MakeVector4_FromFloat(Float=&FN_FrontEndFloatModifier(Gameplay=1, Front End=0.8))
    )
    Lifetime=&UniformRangedFloat(Minimum=0.1, Maximum=0.35)
    Mass Mode=::NewEnumerator1
    Sprite Size=&UniformRangedVector2D(Minimum=<0.7 1>, Maximum=<1 1.5>)
    Sprite Size Mode=::NewEnumerator4
    UsePositionOffset=false
    Write Material Random=true
   ) {module=InitializeParticle001}
   /Niagara/Modules/Spawn/Location/SphereLocation(
    Hemisphere X=true
    MaskSpawnGroup=false
    Module.Override Randomness=false
    Sphere Origin:NiagaraPosition=/Niagara/DynamicInputs/Helpers/SimulationPosition()
    Sphere Radius=5
    Surface Only=false
   )
   &AddVelocity(Velocity=&Multiply_Vector(
     A=$Output.SphereLocation.SphereNormal
     B=&VectorFromFloat(Value=&UniformRangedFloat(Minimum=5, Maximum=20))
    )) {enabled=false, module=AddVelocity001}
   &AddVelocity(
    Cone Angle=25
    Cone Axis=&RotateVectorByQuaternion(
     Vector To Rotate=<0 0 1>
     Quaternion=&OrientMeshToVector(
      Look At Direction=&TransformVector(Vector=$System.SparksDirection, Destination Space=::World)
      Reference Axis Vector=<0 0 1>
     )
    )
    Inner Cone Angle=5
    Module.UseConeFalloff=false
    Module.UseVelDistribution=false
    Override Seed=false
    Rotation Coordinate Space=::World
    Velocity Mode=::NewEnumerator2
    Velocity Speed=&Lerp_Float(
     A=100
     B=550
     Alpha=&FloatFromCurve(
      FloatCurve=DataInterface() {propertyValues {Curve {keys [
          {interpMode=RCIM_Cubic, tangentMode=RCTM_User, time=0, value=0}
          {interpMode=RCIM_Cubic, tangentMode=RCTM_User, time=0.5, value=1}
          {tangentMode=RCTM_User, time=1, value=0}
         ]}}}
      CurveIndex=&RandomRangeFloat(Minimum=0, Maximum=1)
     )
    )
   ) {module=AddVelocity003}
   &AddVelocity(Velocity=&UniformRangedVector(Minimum=<100 0 -200>, Maximum=<1000 0 -200>)) {enabled=false}
   &CameraOffset(Camera Offset Amount=25)
  ]
  particleUpdateScript [
   &ParticleState()
   &AccelerationForce(Acceleration=&RandomRangeVector(Minimum=<0 0 -600>, Maximum=<0 0 -1000>))
   &Collision(
    Age Colliding Particles=false
    KillOnCollision=false
    Module.Control Rest State Via Surface Normal=false
    Module.EnableMaxCollisionCount=false
    Module.Randomize Collision Normal Vector=false
   ) {enabled=false}
   &Drag(Drag=&RandomRangeFloat(Minimum=0.5, Maximum=1.4), Use Linear Drag=true, Use Rotational Drag=true)
   &SolveForcesAndVelocity(Clamp Velocity=false, Limit Acceleration=false)
   &ScaleSpriteSizeBySpeed(Max Scale Factor=<1 5>, Min Scale Factor=<0.7 1>, Velocity Threshold=550)
   &ScaleSpriteSize(Scale Factor=&Vector2DFromFloat(Value=&FloatFromCurve(FloatCurve=DataInterface() {propertyValues {Curve {keys [{arriveTangent=-6.725, interpMode=RCIM_Cubic, leaveTangent=-6.725, tangentMode=RCTM_User, time=0.65, value=1}, <at 1 0 RCIM_Cubic>]}}})))
   &ScaleColor(
    Scale RGB=&MakeVectorFromLinearColorRGB(Color=&MultiplyLinearColor(
      Linear Color=&ColorFromCurve(ColorCurve=DataInterface() {propertyValues {
         AlphaCurve {keys [<at 0 1 RCIM_Cubic>,<at 1 1 RCIM_Cubic>]}
         BlueCurve {keys [<at 0.2 15 RCIM_Cubic>,<at 0.7 0.4427 RCIM_Cubic>]}
         GreenCurve {keys [<at 0.2 0.2486 RCIM_Cubic>,<at 0.7 0.0073 RCIM_Cubic>]}
         RedCurve {keys [<at 0.2 8.3702 RCIM_Cubic>,<at 0.7 0.247 RCIM_Cubic>]}
        }})
      Scale Factor=&MakeVector4_FromFloat(Float=3)
     ))
    ScaleA=true
    ScaleRGB=true
   )
  ]
  rendererClasses [&NiagaraSpriteRendererProperties]
  renderers [{rendererClass=&NiagaraSpriteRendererProperties}]
 }
 NE_FN_PhysMatDebris {
  emitterUpdateScript [
   &EmitterState()
   &SpawnRate(Spawn Probability=0.67, SpawnRate=5, Use Spawn Probability=true)
   &SpawnBurst_Instantaneous(
    Spawn Count=&RandomRangeInt(Minimum=2, Maximum=4)
    Use Spawn Probability=false
    UseLoopCountLimit=false
   )
  ]
  particleSpawnScript [
   /SetVariables(Particles.BiasedRandom=&FloatFromCurve(
     FloatCurve=DataInterface() {propertyValues {Curve {keys [
         <at 0.2 1 RCIM_Cubic>
         {arriveTangent=-3.3394, interpMode=RCIM_Cubic, leaveTangent=-3.3394, tangentMode=RCTM_User, time=0.5, value=0.5}
         <at 0.8 0 RCIM_Cubic>
        ]}}}
     CurveIndex=&RandomRangeFloat(Minimum=0, Maximum=1)
    )) {isSetParameters=true}
   &InitializeParticle(
    Color Mode=::NewEnumerator0
    Lifetime=&RandomRangeFloat(Minimum=0.5, Maximum=1.25)
    Mass=&Lerp_Float(A=0.6, B=1.2, Alpha=$Particles.BiasedRandom)
    Mass Mode=::NewEnumerator1
    Mesh Scale=&Multiply_VectorByFloat(
     Vector=&RandomRangeVector(Minimum=<0.9 0.4 0.5>, Maximum=<1 1 1>)
     Float=&Lerp_Float(A=0.7, B=1, Alpha=$Particles.BiasedRandom)
    )
    Mesh Scale Mode=::NewEnumerator4
    Position Offset=<0 0 3>
    UsePositionOffset=true
    Write Material Random=true
   )
   &FN_PhysicalSurface(
    Dirt Mesh Count=2
    Glass Mesh Count=3
    Ice Mesh Count=2
    Metal Gravity=&RandomRangeFloat(Minimum=-350, Maximum=-1800)
    Metal Mesh Count=2
    Metal Velocity=&RandomRangeFloat(Minimum=80, Maximum=375)
    Metal Velocity Cone Angle=130
    Physical Surface=2
    Stone Gravity=&RandomRangeFloat(Minimum=-700, Maximum=-1250)
    Stone Mesh Count=2
    Stone Velocity=&RandomRangeFloat(Minimum=80, Maximum=550)
    Stone Velocity Cone Angle=120
    StoneScale=0.6
    Wood Gravity=&RandomRangeFloat(Minimum=-450, Maximum=-900)
    Wood Mesh Count=2
    Wood Scale=1.5
    Wood Velocity=&RandomRangeFloat(Minimum=80, Maximum=600)
    Wood Velocity Cone Angle=130
   )
   &AddVelocity(
    Cone Angle=25
    Cone Axis=&RotateVectorByQuaternion(
     Vector To Rotate=<0 0 1>
     Quaternion=&OrientMeshToVector(
      Look At Direction=&TransformVector(Vector=<-3 1 0>, Destination Space=::World)
      Reference Axis Vector=<0 0 1>
     )
    )
    Module.UseConeFalloff=false
    Module.UseVelDistribution=false
    Override Seed=false
    Rotation Coordinate Space=::World
    Velocity Mode=::NewEnumerator2
    Velocity Speed=&Lerp_Float(
     A=250
     B=300
     Alpha=&FloatFromCurve(
      FloatCurve=DataInterface() {propertyValues {Curve {keys [
          {interpMode=RCIM_Cubic, tangentMode=RCTM_User, time=0, value=0}
          {interpMode=RCIM_Cubic, tangentMode=RCTM_User, time=0.5, value=1}
          {tangentMode=RCTM_User, time=1, value=0}
         ]}}}
      CurveIndex=&RandomRangeFloat(Minimum=0, Maximum=1)
     )
    )
   ) {module=AddVelocity003}
   &AddVelocityInCone(
    Cone Angle=$Output.FN_PhysicalSurface.SurfaceResponseConeAngle
    Module.Use Velocity Falloff On Cone Axis=false
    Velocity Strength=$Output.FN_PhysicalSurface.SurfaceResponseVelocity
   ) {enabled=false}
   &InitialMeshOrientation(Use Rotation Vector=false)
   &AddRotationalVelocity(Rotation Rate=&RandomVector(Vector Scale=&RandomRangeFloat(Minimum=1, Maximum=2)))
   &ApplyInitialForces()
   /SetVariables(Particles.GravityOverrideValue=$Output.FN_PhysicalSurface.GravityOverride) {isSetParameters=true}
  ]
  particleUpdateScript [
   &ParticleState()
   &GravityForce(Gravity=$Particles.GravityOverrideValue)
   &Collision(
    Advanced Aging Rate=1
    Age Colliding Particles=false
    CPU Collision Trace Channel=ECC_WorldStatic
    Collision Enabled=&Boolean_AndOperation(
     Bool1=&SetBoolByFloatComparison(A=$Particles.Initial.Lifetime, B=1)
     Bool2=&SetBoolByFloatComparison(A=$Particles.Age, B=0.12)
    )
    Control Roll On Collision=true
    Friction=1
    Friction During a Bounce=0.25
    Max CPU Trace Length=100
    Module.Control Rest State Via Surface Normal=false
    Module.EnableMaxCollisionCount=false
    Module.Randomize Collision Normal Vector=true
    Randomize Collision Normal=0.25
    ScaleRateOfRotation=false
   )
   &SolveForcesAndVelocity(Clamp Velocity=false, Limit Acceleration=false)
   &ScaleMeshSize(Scale Factor=&VectorFromFloat(Value=&FloatFromCurve(FloatCurve=DataInterface() {propertyValues {Curve {keys [<at 0.6 1 RCIM_Cubic>, {arriveTangent=-6.308, interpMode=RCIM_Cubic, leaveTangent=-6.308, tangentMode=RCTM_User, time=1, value=0}]}}})))
   &DynamicMaterialParameters(Module.Param0WriteEnabled=true, Module.Param1WriteEnabled=true, Module.Param2WriteEnabled=false, Module.Param3WriteEnabled=false)
  ]
  rendererClasses [&NiagaraSpriteRendererProperties, &NiagaraMeshRendererProperties]
  renderers [
   {rendererClass=&NiagaraSpriteRendererProperties}
   {rendererClass=&NiagaraMeshRendererProperties, rendererIndex=1}
   {rendererClass=&NiagaraMeshRendererProperties, rendererIndex=2}
   {rendererClass=&NiagaraMeshRendererProperties, rendererIndex=3}
   {rendererClass=&NiagaraMeshRendererProperties, rendererIndex=4}
   {rendererClass=&NiagaraMeshRendererProperties, rendererIndex=5}
   {rendererClass=&NiagaraMeshRendererProperties, rendererIndex=6}
   {rendererClass=&NiagaraMeshRendererProperties, rendererIndex=7}
   {rendererClass=&NiagaraMeshRendererProperties, rendererIndex=8}
  ]
 }
 Smoke {
  emitterUpdateScript [
   &EmitterState()
   &SpawnRate(Spawn Probability=0.8, SpawnRate=20, Use Spawn Probability=true)
   &SpawnBurst_Instantaneous(Spawn Count=&UniformRangedInt(Minimum=2, Maximum=3), Use Spawn Probability=false)
  ]
  particleSpawnScript [
   &InitializeParticle(
    Color=<rgba 0.4767 0.1919 0.7084 0.65>
    Lifetime=&UniformRangedFloat(Minimum=0.5, Maximum=0.8)
    Mass Mode=::NewEnumerator1
    Sprite Size=&Vector2DFromFloat(Value=&UniformRangedFloat(Minimum=70, Maximum=150))
    Sprite Size Mode=::NewEnumerator4
    UsePositionOffset=false
    Write Material Random=true
   ) {module=InitializeParticle001}
   &CameraOffset(Camera Offset Amount=&UniformRangedFloat(Minimum=0, Maximum=20))
   &SphereLocation(Hemisphere X=true, MaskSpawnGroup=false, Module.Override Randomness=false, Sphere Radius=5, Surface Only=false)
   &AddVelocity(Velocity=&UniformRangedVector(Minimum=<100 0 0>, Maximum=<400 0 0>)) {enabled=false}
   &AddVelocity(
    Cone Angle=15
    Cone Axis=&RotateVectorByQuaternion(
     Vector To Rotate=<0 0 1>
     Quaternion=&OrientMeshToVector(
      Look At Direction=&TransformVector(Vector=$System.SmokeDirection, Destination Space=::World)
      Reference Axis Vector=<0 0 1>
     )
    )
    Module.UseConeFalloff=false
    Module.UseVelDistribution=false
    Override Seed=false
    Rotation Coordinate Space=::World
    Velocity Mode=::NewEnumerator2
    Velocity Speed=&RandomRangeFloat(Minimum=25, Maximum=225)
   ) {module=AddVelocity003}
   &AddVelocity(Velocity=&Multiply_Vector(A=$Output.SphereLocation.SphereNormal, B=&VectorFromFloat(Value=5))) {enabled=false, module=AddVelocity001}
  ]
  particleUpdateScript [
   &ParticleState()
   &Drag(Drag=&RandomRangeFloat(Minimum=2, Maximum=3), Use Linear Drag=true, Use Rotational Drag=true)
   &AccelerationForce(Acceleration=&RandomRangeVector(Minimum=<0 0 5>, Maximum=<0 0 35>))
   &SolveForcesAndVelocity(Clamp Velocity=false, Limit Acceleration=false)
   /SetVariables(Particles.SubImageIndex=&Lerp_Float(A=0, B=10, Alpha=$Particles.NormalizedAge)) {isSetParameters=true}
   &ScaleSpriteSize(Scale Factor=&Vector2DFromFloat(Value=&FloatFromCurve(FloatCurve=DataInterface() {propertyValues {Curve {keys [
          {arriveTangent=0.221, interpMode=RCIM_Cubic, leaveTangent=0.221, tangentMode=RCTM_User, time=0, value=0.15}
          {arriveTangent=0.7659, interpMode=RCIM_Cubic, leaveTangent=0.7659, tangentMode=RCTM_User, time=1, value=0.8}
         ]}}})))
   &DynamicMaterialParameters(
    Index 0 Param 1=&FloatFromCurve(FloatCurve=DataInterface() {propertyValues {Curve {keys [<at 0.5 0.5>,<at 1 -0.1>]}}})
    Index 0 Param 4=1
    Module.Param0WriteEnabled=true
    Module.Param1WriteEnabled=true
    Module.Param2WriteEnabled=true
    Module.Param3WriteEnabled=true
   )
   &ScaleColor(
    Scale Alpha=&FloatFromCurve(
     FloatCurve=DataInterface() {propertyValues {Curve {keys [<at 0 0>,<at 0.25 1>,<at 1 0>]}}}
     Scale Curve=&RandomRangeFloat(Minimum=0.15, Maximum=0.1)
    )
    ScaleA=true
    ScaleRGB=true
   )
  ]
  rendererClasses [&NiagaraSpriteRendererProperties]
  renderers [{rendererClass=&NiagaraSpriteRendererProperties}]
 }
 SmokePuff_Gen5 {
  bEnabled=false
  emitterUpdateScript [
   &EmitterState()
   &SpawnBurst_Instantaneous(
    Spawn Count=&RandomRangeInt(Minimum=2, Maximum=3)
    Use Spawn Probability=false
    UseLoopCountLimit=false
   )
  ]
  particleSpawnScript [
   &InitializeParticle(
    Color Maximum=<rgba 0.76 1 1 0.2>
    Color Minimum=<rgba 0.8 0.917 1 0.15>
    Color Mode=::NewEnumerator2
    Lifetime Max=0.4
    Lifetime Min=0.3
    Lifetime Mode=::NewEnumerator1
    Sprite Size Mode=::NewEnumerator1
    Uniform Sprite Size Max=150
    Uniform Sprite Size Min=100
    UsePositionOffset=false
    Write Material Random=true
   )
   &CameraOffset(Camera Offset Amount=&RandomRangeFloat(Minimum=40, Maximum=50, Evaluation Type=::NewEnumerator1))
   &ShapeLocation(
    MaskSpawnGroup=false
    Override Seed=false
    Sphere Radius=&Multiply_Float(A=10, B=&FN_FrontEndFloatModifier(Gameplay=1, Front End=$User.FE_Scale))
    UseOwnerScale=true
   )
   &AddVelocity(
    Velocity=&Multiply_VectorByFloat(Vector=$Particles.ShapeLocation.ShapeVector, Float=10)
    Velocity Speed Scale=&FN_FrontEndFloatModifier(Gameplay=1, Front End=$User.FE_Scale)
   )
  ]
  particleUpdateScript [
   &ParticleState()
   &ScaleSpriteSize(Uniform Curve Scale=&FN_FrontEndFloatModifier(Gameplay=1, Front End=$User.FE_Scale))
   &ScaleColor(
    Scale Alpha=&FloatFromCurve(FloatCurve=DataInterface() {propertyValues {Curve {keys [<at 0 0>,<at 0.25 0.5>,<at 1 0>]}}})
    Scale RGB=&VectorFromCurve(VectorCurve=DataInterface() {propertyValues {XCurve {keys [<at 0 1>,<at 1 1>]}, YCurve {keys [<at 0 1>,<at 1 1>]}, ZCurve {keys [<at 0 1>,<at 1 1>]}}})
    ScaleA=true
    ScaleRGB=true
   )
   &DynamicMaterialParameters(
    Index 0 Param 1=&Add_Float(
     A=&RandomRangeFloat(Minimum=0, Maximum=1)
     B=&FloatFromCurve(
      FloatCurve=DataInterface() {propertyValues {Curve {keys [
          <at 0 0 RCIM_Cubic>
          {arriveTangent=1.5947, interpMode=RCIM_Cubic, leaveTangent=1.5947, tangentMode=RCTM_User, time=0.2557, value=0.5257}
          <at 1 1 RCIM_Cubic>
         ]}}}
      CurveIndex=$Particles.Age
      Scale Curve=0.25
     )
    )
    Module.Param0WriteEnabled=true
    Module.Param1WriteEnabled=true
    Module.Param2WriteEnabled=true
    Module.Param3WriteEnabled=true
   ) {module=DynamicMaterialParameters001}
   &SolveForcesAndVelocity(Clamp Velocity=false, Limit Acceleration=false)
  ]
  rendererClasses [&NiagaraSpriteRendererProperties]
  renderers [
   {rendererClass=&NiagaraSpriteRendererProperties}
   {rendererClass=&NiagaraSpriteRendererProperties, rendererIndex=1}
  ]
 }
 Sparks001 {
  bEnabled=false
  emitterUpdateScript [
   &EmitterState(
    Enable Distance Culling=true
    Max Distance Response=::KillImmediately
    Module.bApplySpawnCountScaleByDistance=true
    Module.bUseMaxDistance=true
    Module.bUseMinDistance=false
    Scalability Mode=::NewEnumerator1
    Scale Spawn Count=true
    Spawn Count Scale By Distance=DataInterface() {propertyValues {Curve {keys [<at 0.3676 0.9657>,<at 0.9962 0.4982>]}}}
   )
   &SpawnBurst_Instantaneous(Spawn Count=&UniformRangedInt(Minimum=5, Maximum=10), Use Spawn Probability=false)
   &SpawnRate(
    Spawn Probability=&FloatFromCurve(
     FloatCurve=DataInterface() {propertyValues {Curve {keys [{arriveTangent=-12.2872, interpMode=RCIM_Cubic, leaveTangent=-12.2872, tangentMode=RCTM_User, time=0, value=1}, <at 0.1 0>]}}}
     CurveIndex=$System.Age
    )
    SpawnRate=111
    Use Spawn Probability=true
   )
  ]
  particleSpawnScript [
   &InitializeParticle(
    Color=<rgba 3 3 3 1>
    Lifetime Max=0.45
    Lifetime Min=0.166
    Lifetime Mode=::NewEnumerator1
    Sprite Size Mode=::NewEnumerator1
    Uniform Sprite Size Max=8
    Uniform Sprite Size Min=2.5
    UsePositionOffset=false
    Write Material Random=true
   )
   &ShapeLocation(
    MaskSpawnGroup=false
    Override Seed=false
    Sphere Radius=&Multiply_Float(A=9, B=&FN_FrontEndFloatModifier(Gameplay=1, Front End=$User.FE_Scale))
    Sphere Surface Distribution=1
    UseOwnerScale=true
   )
   &AddVelocity(
    Module.Use Falloff Distance=false
    Override Seed=false
    Velocity Mode=::NewEnumerator1
    Velocity Speed=&Multiply_Float(
     A=&RandomRangeFloat(Minimum=222, Maximum=666)
     B=&FN_FrontEndFloatModifier(Gameplay=1, Front End=$User.FE_Scale)
    )
   )
   &DynamicMaterialParameters(
    Index 0 Param 1=3
    Index 0 Param 2=20
    Index 0 Param 3=1
    Module.Param0WriteEnabled=true
    Module.Param1WriteEnabled=true
    Module.Param2WriteEnabled=true
    Module.Param3WriteEnabled=true
   )
  ]
  particleUpdateScript [
   &ParticleState()
   &GravityForce(Gravity=<0 0 -666>)
   &Drag(Drag=&UniformRangedFloat(Minimum=2, Maximum=6), Use Linear Drag=true, Use Rotational Drag=false)
   &SolveForcesAndVelocity(Clamp Velocity=false, Limit Acceleration=false)
   &ScaleSpriteSize(Scale Factor=&Vector2DFromCurve(
     Vector2Curve=DataInterface() {propertyValues {
       XCurve {keys [
         {arriveTangent=5.9728, interpMode=RCIM_Cubic, leaveTangent=5.9728, tangentMode=RCTM_User, time=0, value=0.1}
         {arriveTangent=-0.1, interpMode=RCIM_Cubic, leaveTangent=-0.1, tangentMode=RCTM_User, time=0.18, value=0.8}
         {interpMode=RCIM_Cubic, tangentMode=RCTM_User, time=1, value=0}
        ]}
       YCurve {keys [
         {arriveTangent=6.2715, interpMode=RCIM_Cubic, leaveTangent=6.2715, tangentMode=RCTM_User, time=0, value=0.1}
         {arriveTangent=-0.1001, interpMode=RCIM_Cubic, leaveTangent=-0.1001, tangentMode=RCTM_User, time=0.18, value=0.8}
         {interpMode=RCIM_Cubic, tangentMode=RCTM_User, time=0.9988, value=0}
        ]}
      }}
     ScaleCurve=&Vector2DFromFloat(Value=&FN_FrontEndFloatModifier(Gameplay=1, Front End=$User.FE_Scale))
    ))
   &ScaleSpriteSizeBySpeed(Max Scale Factor=<0.5 8>, Min Scale Factor=<0.5 1>, Velocity Threshold=666)
   &ScaleColor(
    Linear Color Curve=DataInterface() {propertyValues {
      AlphaCurve {keys [<at 0.333 1>,<at 1 0>]}
      BlueCurve {keys [<at 0 0.24>,<at 0.166 0.2>,<at 0.333 0>,<at 0.75 0>]}
      GreenCurve {keys [<at 0 3.12>,<at 0.166 2.6>,<at 0.333 0.6667>,<at 0.75 0.15>]}
      RedCurve {keys [<at 0 6>,<at 0.166 5>,<at 0.333 4>,<at 0.75 3>]}
     }}
    Scale Mode=::NewEnumerator2
   )
  ]
  rendererClasses [&NiagaraSpriteRendererProperties]
  renderers [{rendererClass=&NiagaraSpriteRendererProperties}]
 }
 Sparks_Directional {
  bEnabled=false
  emitterUpdateScript [
   &EmitterState(
    Enable Distance Culling=true
    Max Distance Response=::KillImmediately
    Module.bApplySpawnCountScaleByDistance=true
    Module.bUseMaxDistance=true
    Module.bUseMinDistance=false
    Scalability Mode=::NewEnumerator1
    Scale Spawn Count=true
    Spawn Count Scale By Distance=DataInterface() {propertyValues {Curve {keys [<at 0.3676 0.9657>,<at 0.9962 0.4982>]}}}
   )
   &SpawnBurst_Instantaneous(
    Spawn Count=&UniformRangedInt(Minimum=4, Maximum=11)
    Spawn Probability=0.5
    Use Spawn Probability=true
    UseLoopCountLimit=false
   )
   /SetVariables(
    Emitter.RandVecY:NiagaraFloat=&RandomRangeFloat(Minimum=-1, Maximum=1)
    Emitter.RandVecZ:NiagaraFloat=&RandomRangeFloat(Minimum=-1, Maximum=1)
   ) {enabled=false, isSetParameters=true}
  ]
  particleSpawnScript [
   &InitializeParticle(
    Color=<rgba 3 3 3 1>
    Lifetime=2
    Sprite Size Mode=::NewEnumerator1
    Uniform Sprite Size Max=8
    Uniform Sprite Size Min=2.5
    UsePositionOffset=false
    Write Material Random=true
   )
   &AddVelocity(
    Cone Axis=&MakeVector(X=0, Y=$Emitter.RandVecY, Z=$Emitter.RandVecZ)
    Distribution Along Cone Axis=0.666
    Module.UseConeFalloff=true
    Module.UseVelDistribution=true
    Override Seed=false
    Speed Falloff From Cone Axis=0.666
    Velocity Mode=::NewEnumerator2
    Velocity Speed=&RandomRangeFloat(Minimum=111, Maximum=999)
   ) {enabled=false}
   &AddVelocity(
    Cone Angle=10
    Cone Axis=&RotateVectorByQuaternion(
     Vector To Rotate=<0 0 1>
     Quaternion=&OrientMeshToVector(
      Look At Direction=&TransformVector(
       Vector=&FN_DirectionalVectorfromMontageIndex(
        Montage Section Index:NiagaraInt32=$User.SwingSectionIndex
        Montage Index 1:Vector3f=<0 1 1>
        Montage Index 2:Vector3f=<-1 1 0.9506>
        Montage Index 3:Vector3f=<2.6556 -1 2.3295>
        Montage Index 4:Vector3f=<2.4576 -1 -4.0592>
        Invalid Montage Index:Vector3f=<1 0 0>
       )
       Destination Space=::World
      )
      Reference Axis Vector=<0 0 1>
     )
    )
    Module.UseConeFalloff=false
    Module.UseVelDistribution=false
    Override Seed=false
    Rotation Coordinate Space=::World
    Velocity Mode=::NewEnumerator2
    Velocity Speed=&RandomRangeFloat(Minimum=400, Maximum=1000)
   ) {module=AddVelocity001}
   &DynamicMaterialParameters(
    Index 0 Param 1=3
    Index 0 Param 3=1
    Module.Param0WriteEnabled=true
    Module.Param1WriteEnabled=true
    Module.Param2WriteEnabled=true
    Module.Param3WriteEnabled=true
   )
  ]
  particleUpdateScript [
   &ParticleState()
   &GravityForce(Gravity=<0 0 -666>)
   &Drag(Drag=&UniformRangedFloat(Minimum=1, Maximum=3), Use Linear Drag=true, Use Rotational Drag=false)
   &SolveForcesAndVelocity(Clamp Velocity=false, Limit Acceleration=false)
   &ScaleSpriteSize(Scale Factor=&Vector2DFromCurve(
     Vector2Curve=DataInterface() {propertyValues {
       XCurve {keys [
         {arriveTangent=5.9728, interpMode=RCIM_Cubic, leaveTangent=5.9728, tangentMode=RCTM_User, time=0, value=0.1}
         {arriveTangent=-0.1, interpMode=RCIM_Cubic, leaveTangent=-0.1, tangentMode=RCTM_User, time=0.18, value=0.8}
         {interpMode=RCIM_Cubic, tangentMode=RCTM_User, time=1, value=0}
        ]}
       YCurve {keys [
         {arriveTangent=6.2715, interpMode=RCIM_Cubic, leaveTangent=6.2715, tangentMode=RCTM_User, time=0, value=0.1}
         {arriveTangent=-0.1001, interpMode=RCIM_Cubic, leaveTangent=-0.1001, tangentMode=RCTM_User, time=0.18, value=0.8}
         {interpMode=RCIM_Cubic, tangentMode=RCTM_User, time=0.9988, value=0}
        ]}
      }}
     ScaleCurve=<1 1>
    ))
   &ScaleSpriteSizeBySpeed(Max Scale Factor=<0.333 11>, Min Scale Factor=<0.5 1>, Velocity Threshold=666)
   &ScaleColor(
    Linear Color Curve=DataInterface() {propertyValues {
      AlphaCurve {keys [<at 0 0>,<at 0.166 1>,<at 1 0>]}
      BlueCurve {keys [<at 0 0.24>,<at 0.166 0.2>,<at 0.333 0>,<at 0.75 0>]}
      GreenCurve {keys [<at 0 3.12>,<at 0.166 2.6>,<at 0.333 0.6667>,<at 0.75 0.15>]}
      RedCurve {keys [<at 0 6>,<at 0.166 5>,<at 0.333 4>,<at 0.75 3>]}
     }}
    Scale Mode=::NewEnumerator2
   )
  ]
  rendererClasses [&NiagaraSpriteRendererProperties]
  renderers [{rendererClass=&NiagaraSpriteRendererProperties}]
 }
}
summary {
 systemName=NS_Emote_PeacefulPoemPumped_Sparks
 emitters [
  {emitterName=Flash_Flare, rendererClasses [&NiagaraSpriteRendererProperties, &NiagaraLightRendererProperties]}
  {emitterName=Lazy_Sparks, rendererClasses [&NiagaraSpriteRendererProperties]}
  {emitterName=Smoke, rendererClasses [&NiagaraSpriteRendererProperties]}
  {emitterName=Glow, rendererClasses [&NiagaraSpriteRendererProperties]}
  {bEnabled=false, emitterName=SmokePuff_Gen5, rendererClasses [&NiagaraSpriteRendererProperties]}
  {
   bEnabled=false
   emitterName=Flash_Flare001
   rendererClasses [&NiagaraSpriteRendererProperties, &NiagaraLightRendererProperties]
  }
  {bEnabled=false, emitterName=Sparks001, rendererClasses [&NiagaraSpriteRendererProperties]}
  {bEnabled=false, emitterName=Sparks_Directional, rendererClasses [&NiagaraSpriteRendererProperties]}
  {bEnabled=false, emitterName=DirectionalSparks, rendererClasses [&NiagaraSpriteRendererProperties]}
  {
   emitterName=NE_FN_PhysMatDebris
   rendererClasses [&NiagaraSpriteRendererProperties, &NiagaraMeshRendererProperties]
  }
 ]
 userVariables [
  User.bIsFrontEnd {type=NiagaraBool}
  User.bIsFrontEndPreview {type=NiagaraBool}
  User.ColorSwitcher {type=NiagaraFloat, defaultValue=0, description=""}
  User.FE_Scale {type=NiagaraFloat, defaultValue=0, description=""}
  User.PhysicalSurface {type=NiagaraInt32}
  User.SwingSectionIndex {type=NiagaraInt32}
 ]
}
